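; Player component that gives its owner a gameplay effect and hero rank score boost while participating in a Spire event.
; On start, it applies GE_WhileInSpireEvent to the owner's Fort ability system, then, on authority only, pushes a 2x CherrySmoke hero score multiplier, storing both handles.
; On stop, authority uses those handles to remove both, and SpirePlayspace is cleared on all sides.
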(event EventBeginPlay)

(event EventTick
  (params (DeltaSeconds :type Float%20%28single-precision%29)))

(event EventStartedParticipatingInSpireEvent
  (params (SpireBGA :type Building%20Gameplay%20Actor%20Object%20Reference))
  ;[0.0142 0.06854 0.15 0.5] Applying/removing GE to players that enter/exit
  (bind _returnvalue (GameplayEffects|ApplyGameplayEffectToSelf :self (Ability|GetFortAbilitySystemComponent (Components|GetOwner)) :GameplayEffectClass "/SpiresGameplay/Gameplay/GameplayEffects/GE_WhileInSpireEvent.GE_WhileInSpireEvent_C" :Level 0.0))
  (Variables|Default|SetWhileInSpireEventGEHandle _returnvalue)
  ;[0.15 0.1232 0.007516 0.5] Applying initial multiplier of 2x when entering the space. Clearing when exiting
  (Utilities|SwitchHasAuthority
    (:Authority
      (bind _returnvalue_2 (Class|CherrySmokeHeroRankingManager|PushHeroScoreMultiplier (Actor|GetComponentbyClass (Components|GetOwner) "/Script/CherrySmokeHeroRankingSystemRuntime.CherrySmokeHeroRankingManager") (Utilities|Struct|MakeHeroRankMultiplier 2.0)))
      (Variables|Default|SetHeroRankMultiplierHandle _returnvalue_2))))

(event EventStoppedParticipatingInSpireEvent
  (params (SpireBGA :type Building%20Gameplay%20Actor%20Object%20Reference))
  ;[0.0142 0.06854 0.15 0.5] Applying/removing GE to players that enter/exit
  (Utilities|SwitchHasAuthority
    (:Authority
      (GameplayEffects|RemoveActiveGameplayEffect (Ability|GetFortAbilitySystemComponent (Components|GetOwner)) (Variables|Default|GetWhileInSpireEventGEHandle) 1)
      ;[0.15 0.1232 0.007516 0.5] Applying initial multiplier of 2x when entering the space. Clearing when exiting
      (Class|CherrySmokeHeroRankingManager|RemoveHeroScoreMultiplier (Actor|GetComponentbyClass (Components|GetOwner) "/Script/CherrySmokeHeroRankingSystemRuntime.CherrySmokeHeroRankingManager") (Variables|Default|GetHeroRankMultiplierHandle)))
    (:Remote
      _))
  (Variables|Default|SetSpirePlayspace 0))
